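# Legacy master material for a low-poly turntable demo, a Default Lit surface enabled for skeletal meshes and mesh deformers.
# It multiplies a wireframe texture by an ambient occlusion texture, giving an occlusion-shaded wireframe look, alongside a separate normal map sample.
# All three textures are exposed as parameters for material instances and default to Nightmare body maps.
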Material /Game/FN_PublicTurntable/Turntable_Lowpoly/Content/Demo/Materials/M_Turntable_Master_Legacy
graph {
 properties {bUsedWithMeshDeformer=true, bUsedWithSkeletalMesh=true, pixelDepthOffsetMode=PDOM_Legacy, usedShadingModels=Default Lit}
 nodes [
  Multiply(
   A=TextureSampleParameter2D() {parameterName=Wireframe, texture=/Game/FN_PublicTurntable/Turntable_Lowpoly/Content/Demo/Textures/Nightmare_Body_Wireframe}.RGB
   B=TextureSampleParameter2D() {parameterName=AO, texture=/Game/FN_PublicTurntable/Turntable_Lowpoly/Content/Demo/Textures/Nightmare_Body_ao_2k}.RGB
  )
  TextureSampleParameter2D() {
   parameterName=Normals
   samplerType=SAMPLERTYPE_Normal
   texture=/Game/FN_PublicTurntable/Turntable_Lowpoly/Content/Demo/Textures/Nightmare_Body_normal_2k
  }
 ]
}
